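# Master material for mirror glass in the Creative glass gallery environment meshes, using the Default Lit shading model.
# A Fresnel term with an adjustable exponent blends inner and outer color parameters to give a view-angle-dependent tint.
# Scalar parameters expose metallic, specular and roughness, defaulting to a fully metallic, low-roughness surface.
# The graph samples no textures, so the mirror look comes only from these parameters.
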Material /Game/Creative/Environments/Meshes/Glass/Materials/M_CP_GlassGallery_Mirror_Master
graph {
 properties {pixelDepthOffsetMode=PDOM_Legacy, usedShadingModels=Default Lit}
 nodes [
  LinearInterpolate(
   A=VectorParameter() {group=Color, parameterName=Glass Inner Color, defaultValue=<rgba 0.5 0.5 0.5 1>}.RGB
   B=VectorParameter() {group=Color, parameterName=Glass Outer Color, defaultValue=<rgba 0 0 0 1>}.RGB
   Alpha=Fresnel(ExponentIn=ScalarParameter() {defaultValue=2, group=Color, parameterName=Color Fresnel Exponent})
  )
  ScalarParameter() {defaultValue=1, parameterName=Metallic}
  ScalarParameter() {defaultValue=0.5, parameterName=Specular}
  ScalarParameter() {defaultValue=0.1, parameterName=Roughness}
 ]
}
